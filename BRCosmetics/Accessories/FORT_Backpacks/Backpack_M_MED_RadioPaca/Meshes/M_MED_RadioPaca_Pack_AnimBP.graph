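# Animation Blueprint for the RadioPaca back bling, built to vibrate its speakers.
# Each update, a macro sums a game-time sine and random noise, clamps it, and fills X of big, medium and small speaker vectors.
# The update also caches music amplitude from an audio-reactive Material Parameter Collection.
# The AnimGraph chains five additive bone-space Modify Bone nodes on speaker bones, with no visible links to those variables.
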AnimBlueprint /BRCosmetics/Accessories/FORT_Backpacks/Backpack_M_MED_RadioPaca/Meshes/M_MED_RadioPaca_Pack_AnimBP
target_skeleton=&/M_MED_RadioPaca_Pack_Skeleton
graphs [
 AnimGraph {nodes [
   n1 = "Transform (Modify) Bone - Bone: dyn_speaker_c"(ComponentPose=Local To Component(LocalPose=Input Pose()), Alpha=1, Translation="0,0,0", Rotation="0,0,0", Scale="1,1,1") {translationMode=BMM_Additive, translationSpace=BCS_BoneSpace, alphaBoolBlend {blendOption=Linear}, boneToModify {boneName=dyn_speaker_c}}
   n2 = "Transform (Modify) Bone - Bone: dyn_speaker_l_01"(ComponentPose=@n1, Alpha=1, Translation="0,0,0", Rotation="0,0,0", Scale="1,1,1") {translationMode=BMM_Additive, translationSpace=BCS_BoneSpace, alphaBoolBlend {blendOption=Linear}, boneToModify {boneName=dyn_speaker_l_01}}
   n3 = "Transform (Modify) Bone - Bone: dyn_speaker_l_02"(ComponentPose=@n2, Alpha=1, Translation="0,0,0", Rotation="0,0,0", Scale="1,1,1") {translationMode=BMM_Additive, translationSpace=BCS_BoneSpace, alphaBoolBlend {blendOption=Linear}, boneToModify {boneName=dyn_speaker_l_02}}
   n4 = "Transform (Modify) Bone - Bone: dyn_speaker_r_01"(ComponentPose=@n3, Alpha=1, Translation="0,0,0", Rotation="0,0,0", Scale="1,1,1") {translationMode=BMM_Additive, translationSpace=BCS_BoneSpace, alphaBoolBlend {blendOption=Linear}, boneToModify {boneName=dyn_speaker_r_01}}
   n5 = "Transform (Modify) Bone - Bone: dyn_speaker_r_02"(ComponentPose=@n4, Alpha=1, Translation="0,0,0", Rotation="0,0,0", Scale="1,1,1") {translationMode=BMM_Additive, translationSpace=BCS_BoneSpace, alphaBoolBlend {blendOption=Linear}, boneToModify {boneName=dyn_speaker_r_02}}
   Output Pose(Result=Component To Local(ComponentPose=@n5)) {layerGroup=None}
  ]}
 CalculateVIbration {nodes [
   n1 = Inputs!() -> (Frequency:double=~, Amplitude:double=~, NoiseAmplitude:double=~)
   n2 = RandomFloatInRange(Min=float * float(A=@n1.NoiseAmplitude, B=-1) {is=title K2Node_PromotableOperator}, Max=@n1.NoiseAmplitude)
   n3 = GetGameTimeInSeconds()
   n4 = float * float(A="Sin (Radians)"(A=float * float(A=@n1, B=@n3)), B=@n1.Amplitude)
   Outputs!(Result:double="Clamp (Float)"(Value=float + float(A=@n4, B=@n2), Min=-1, Max=10))
  ]}
 EventGraph {nodes [
   Event BlueprintUpdateAnimation()
   > Set BigSpeakers(BigSpeakers:Vector=MakeVector(
     X=Calculate VIbration!(Frequency:double=40, Amplitude:double=1.2, NoiseAmplitude:double=0.1) -> (Result:double=~) {is=title K2Node_MacroInstance}
     Y=0
    )) -> (Output_Get:Vector)
   > Set MediumSpeakers(MediumSpeakers:Vector=MakeVector(X=Calculate VIbration(Frequency=40, Amplitude=1, NoiseAmplitude=0.075) -> (Result=~), Y=0)) -> (Output_Get:Vector)
   > Set SmallSpeakers(SmallSpeakers:Vector=MakeVector(X=Calculate VIbration(Frequency=40, Amplitude=0.5, NoiseAmplitude=0.05) -> (Result=~), Y=0)) -> (Output_Get:Vector)
   > n1 = GetScalarParameterValue!(
    execute
    Collection:Material Parameter Collection Object Reference=Get MPC_Audio_Reactive_Parameters() -> (MPC_Audio_Reactive_Parameters:Material Parameter Collection Object Reference)
    ParameterName:Name=Get Amplitude() -> (Amplitude:Name=None)
   ) -> (then, ReturnValue:float) {is=title K2Node_CallMaterialParameterCollectionFunction}
   > Set ReactiveToMusic(ReactiveToMusic:float=@n1) -> (Output_Get:float)
  ]}
]
variables [
 Amplitude {type=string, default=Amplitude_Music}
 BigSpeakers {type=Vector}
 MediumSpeakers {type=Vector}
 MPC_Audio_Reactive_Parameters {type=MaterialParameterCollection, default=/Game/Athena/Events/AudioAnalysis/MPC_Audio_Reactive_Parameters}
 ReactiveToMusic {type=number}
 SmallSpeakers {type=Vector}
]
